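# Conversation database for the RedOasisGooseberry NPC, a soccer-themed character who greets players and offers quests and services.
# The entry point leads to one Speech node with a default cheer and contextual lines chosen by priority curves.
# Contextual lines react to a first-ever conversation or to the player's outfit via HasCID checks for fellow RedOasis, lookalike, soccer and zombie soccer skins.
# The speech then links out to the shared NPCLibrary Quests and BasicServices conversation entries.
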ConversationDatabase /NPCLibrary/NPCs/RedOasisGooseberry/NPCLibrary_Conversation_RedOasisGooseberry
entry_points [{tag=Conversation.Entry.RedOasisGooseberry, destinations [#1]}]
nodes [
 ConversationEntryPointNode_7 {
  id=#1
  class=ConversationEntryPointNode
  outputs [#2]
  properties {entryTag {tagName=Conversation.Entry.RedOasisGooseberry}}
 }
 Speech_C_4 {
  id=#2
  class=Speech_C
  outputs [#3, #4]
  properties {
   nodeName=""
   generalConfig {
    defaultMessage="I hope you're cheering for Fortnite FC! "
    contextualMessages [
     {
      message="You're doing great out there!"
      contextRequirements [{requirement {class=FortControllerRequirement_IsFirstEverConversationWithNPC}}]
      priority {curve {rowName=Tandem.Speech.Priority.FirstEverInteraction}}
     }
     {
      message=It's good to see another pro on the field.
      requirementMatchPolicy=RequireAny
      contextRequirements [{requirement {
         class=FortControllerRequirement_HasCID
         properties {allowedCIds [
           /BRCosmetics/Athena/Items/Cosmetics/Characters/Character_RedOasisApricot
           /BRCosmetics/Athena/Items/Cosmetics/Characters/Character_RedOasisBlackberry
           /BRCosmetics/Athena/Items/Cosmetics/Characters/Character_RedOasisJackfruit
           /BRCosmetics/Athena/Items/Cosmetics/Characters/Character_RedOasisPomegranate
          ]}
        }}]
      priority {curve {rowName=Tandem.Speech.Priority.OutfitReaction}}
     }
     {
      message="Hey, you're me!"
      requirementMatchPolicy=RequireAny
      contextRequirements [{requirement {
         class=FortControllerRequirement_HasCID
         properties {allowedCIds [
           /BRCosmetics/Athena/Items/Cosmetics/Characters/CID_A_078_Athena_Commando_M_Bicycle
           /BRCosmetics/Athena/Items/Cosmetics/Characters/Character_RedOasisGooseberry
          ]}
        }}]
      priority {curve {rowName=Tandem.Speech.Priority.OutfitReaction}}
     }
     {
      message=Nice footwork.
      requirementMatchPolicy=RequireAny
      contextRequirements [{requirement {
         class=FortControllerRequirement_HasCID
         properties {allowedCIds [
           /BRCosmetics/Athena/Items/Cosmetics/Characters/CID_144_Athena_Commando_M_SoccerDudeA
           /BRCosmetics/Athena/Items/Cosmetics/Characters/CID_145_Athena_Commando_M_SoccerDudeB
           /BRCosmetics/Athena/Items/Cosmetics/Characters/CID_146_Athena_Commando_M_SoccerDudeC
           /BRCosmetics/Athena/Items/Cosmetics/Characters/CID_147_Athena_Commando_M_SoccerDudeD
           /BRCosmetics/Athena/Items/Cosmetics/Characters/CID_148_Athena_Commando_F_SoccerGirlA
           /BRCosmetics/Athena/Items/Cosmetics/Characters/CID_149_Athena_Commando_F_SoccerGirlB
           /BRCosmetics/Athena/Items/Cosmetics/Characters/CID_150_Athena_Commando_F_SoccerGirlC
           /BRCosmetics/Athena/Items/Cosmetics/Characters/CID_151_Athena_Commando_F_SoccerGirlD
          ]}
        }}]
      priority {curve {rowName=Tandem.Speech.Priority.OutfitReaction}}
     }
     {
      message=Uh... You look a little green...
      requirementMatchPolicy=RequireAny
      contextRequirements [{requirement {
         class=FortControllerRequirement_HasCID
         properties {allowedCIds [
           /BRCosmetics/Athena/Items/Cosmetics/Characters/CID_589_Athena_Commando_M_SoccerZombieA
           /BRCosmetics/Athena/Items/Cosmetics/Characters/CID_590_Athena_Commando_M_SoccerZombieB
           /Game/Athena/Items/Cosmetics/Characters/CID_591_Athena_Commando_M_SoccerZombieC
           /Game/Athena/Items/Cosmetics/Characters/CID_592_Athena_Commando_M_SoccerZombieD
           /BRCosmetics/Athena/Items/Cosmetics/Characters/CID_593_Athena_Commando_F_SoccerZombieA
           /BRCosmetics/Athena/Items/Cosmetics/Characters/CID_594_Athena_Commando_F_SoccerZombieB
           /BRCosmetics/Athena/Items/Cosmetics/Characters/CID_595_Athena_Commando_F_SoccerZombieC
           /BRCosmetics/Athena/Items/Cosmetics/Characters/CID_596_Athena_Commando_F_SoccerZombieD
          ]}
        }}]
      priority {curve {rowName=Tandem.Speech.Priority.OutfitReaction}}
     }
    ]
   }
  }
 }
 ConversationLinkNode_0 {id=#4, class=ConversationLinkNode, properties {remoteEntryTag {tagName=Conversation.Entry.BasicServices}}}
 ConversationLinkNode_1 {id=#3, class=ConversationLinkNode, properties {remoteEntryTag {tagName=Conversation.Entry.NPCLibrary.Quests}}}
]
